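# Physics asset for the right arm of the Tomato Knight player body, covering five dynamic sleeve flap bones.
# Each flap bone has an identical small capsule body, and a smaller capsule on the lower-arm twist bone serves as their shared anchor.
# Each flap is constrained directly to that twist bone rather than chained to its neighbors, with tight swing and twist limits that allow only slight motion.
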PhysicsAsset /BRCosmetics/Characters/Player/Male/Medium/Bodies/M_MED_Tomato_Knight/Meshes/M_MED_Tomato_Knight_Physics_ArmRight
bodies {
 dyn_sleeveFlap_01_r [{shapeType=Capsule, length=3, radius=2.5, center=<1.5 0 0>, rotation=<rot 90 0 0>}]
 dyn_sleeveFlap_02_r [{shapeType=Capsule, length=3, radius=2.5, center=<1.5 0 0>, rotation=<rot 90 0 0>}]
 dyn_sleeveFlap_03_r [{shapeType=Capsule, length=3, radius=2.5, center=<1.5 0 0>, rotation=<rot 90 0 0>}]
 dyn_sleeveFlap_04_r [{shapeType=Capsule, length=3, radius=2.5, center=<1.5 0 0>, rotation=<rot 90 0 0>}]
 dyn_sleeveFlap_05_r [{shapeType=Capsule, length=3, radius=2.5, center=<1.5 0 0>, rotation=<rot 90 0 0>}]
 lowerarm_twist_02_r [{shapeType=Capsule, length=1, radius=1}]
}
constraints [
 {bone1Name=dyn_sleeveFlap_01_r, bone2Name=lowerarm_twist_02_r, swing1LimitDegrees=5, swing2LimitDegrees=5, twistLimitDegrees=2}
 {bone1Name=dyn_sleeveFlap_02_r, bone2Name=lowerarm_twist_02_r, swing1LimitDegrees=5, swing2LimitDegrees=5, twistLimitDegrees=2}
 {bone1Name=dyn_sleeveFlap_03_r, bone2Name=lowerarm_twist_02_r, swing1LimitDegrees=5, swing2LimitDegrees=5, twistLimitDegrees=2}
 {bone1Name=dyn_sleeveFlap_04_r, bone2Name=lowerarm_twist_02_r, swing1LimitDegrees=5, swing2LimitDegrees=5, twistLimitDegrees=2}
 {bone1Name=dyn_sleeveFlap_05_r, bone2Name=lowerarm_twist_02_r, swing1LimitDegrees=5, swing2LimitDegrees=5, twistLimitDegrees=2}
]
