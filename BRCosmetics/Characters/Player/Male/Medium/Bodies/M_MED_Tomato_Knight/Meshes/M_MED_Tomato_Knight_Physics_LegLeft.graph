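# Physics asset for the left leg of the Tomato Knight male medium body, covering secondary motion of its ankle cloth and leg pouch.
# Token-sized thigh and calf capsules act as anchors, joined by a constraint with free swing and twist.
# Four capsules around the ankle constrain to the calf with moderate swing and narrow twist limits.
# The pouch box constrains tightly to the thigh, with one swing axis locked.
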PhysicsAsset /BRCosmetics/Characters/Player/Male/Medium/Bodies/M_MED_Tomato_Knight/Meshes/M_MED_Tomato_Knight_Physics_LegLeft
bodies {
 calf_l [{shapeType=Capsule, length=1, radius=1}]
 dyn_clothAnkleBack_l [{shapeType=Capsule, length=3, radius=3, center=<1.5 0 0>, rotation=<rot 90 0 0>}]
 dyn_clothAnkleFront_l [{shapeType=Capsule, length=3, radius=3, center=<1.5 0 0>, rotation=<rot 90 0 0>}]
 dyn_clothAnkleLeft_l [{shapeType=Capsule, length=3, radius=3, center=<1.5 0 0>, rotation=<rot 90 0 0>}]
 dyn_clothAnkleRight_l [{shapeType=Capsule, length=3, radius=3, center=<1.5 0 0>, rotation=<rot 90 0 0>}]
 dyn_pouchLeg_01 [{shapeType=Box, extentX=14, extentY=11, extentZ=4, center=<2.4959 0 1.0369>}]
 thigh_l [{shapeType=Capsule, length=1, radius=1}]
}
constraints [
 {bone1Name=dyn_pouchLeg_01, bone2Name=thigh_l, swing1LimitDegrees=5, swing1Motion=Locked, swing2LimitDegrees=3, twistLimitDegrees=1}
 {
  bone1Name=calf_l
  bone2Name=thigh_l
  swing1LimitDegrees=45
  swing1Motion=Free
  swing2LimitDegrees=45
  swing2Motion=Free
  twistLimitDegrees=45
  twistMotion=Free
 }
 {bone1Name=dyn_clothAnkleFront_l, bone2Name=calf_l, swing1LimitDegrees=15, swing2LimitDegrees=15, twistLimitDegrees=5}
 {bone1Name=dyn_clothAnkleLeft_l, bone2Name=calf_l, swing1LimitDegrees=15, swing2LimitDegrees=15, twistLimitDegrees=5}
 {bone1Name=dyn_clothAnkleBack_l, bone2Name=calf_l, swing1LimitDegrees=15, swing2LimitDegrees=15, twistLimitDegrees=5}
 {bone1Name=dyn_clothAnkleRight_l, bone2Name=calf_l, swing1LimitDegrees=15, swing2LimitDegrees=15, twistLimitDegrees=5}
]
