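# Physics asset for the right leg of the Tomato Knight player character body, simulating the cloth around the right ankle.
# Four capsule bodies on the front, back, left and right ankle cloth bones each attach directly to a small calf capsule.
# Their constraints share tight swing and twist limits, keeping the motion subtle, with no constraints between neighboring cloth pieces.
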PhysicsAsset /BRCosmetics/Characters/Player/Male/Medium/Bodies/M_MED_Tomato_Knight/Meshes/M_MED_Tomato_Knight_Physics_LegRight
bodies {
 calf_r [{shapeType=Capsule, length=1, radius=1}]
 dyn_ClothAnkleBack_r [{shapeType=Capsule, length=3, radius=3, center=<1.5 0 0>, rotation=<rot 90 0 0>}]
 dyn_ClothAnkleFront_r [{shapeType=Capsule, length=3, radius=3, center=<1.5 0 0>, rotation=<rot 90 0 0>}]
 dyn_ClothAnkleLeft_r [{shapeType=Capsule, length=3, radius=3, center=<1.5 0 0>, rotation=<rot 90 0 0>}]
 dyn_ClothAnkleRight_r [{shapeType=Capsule, length=3, radius=3, center=<1.5 0 0>, rotation=<rot 90 0 0>}]
}
constraints [
 {bone1Name=dyn_ClothAnkleFront_r, bone2Name=calf_r, swing1LimitDegrees=15, swing2LimitDegrees=15, twistLimitDegrees=5}
 {bone1Name=dyn_ClothAnkleLeft_r, bone2Name=calf_r, swing1LimitDegrees=15, swing2LimitDegrees=15, twistLimitDegrees=5}
 {bone1Name=dyn_ClothAnkleBack_r, bone2Name=calf_r, swing1LimitDegrees=15, swing2LimitDegrees=15, twistLimitDegrees=5}
 {bone1Name=dyn_ClothAnkleRight_r, bone2Name=calf_r, swing1LimitDegrees=15, swing2LimitDegrees=15, twistLimitDegrees=5}
]
