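# Physics asset for the three pouches on the Tomato Knight male medium character body, covering only the pelvis and pouch bones.
# Each pouch has a capsule body constrained to its tiny master-bone capsule with very tight swing and twist limits.
# Each master body is constrained to the small pelvis capsule with free swing and twist, so each pouch chain can rotate about that joint.
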PhysicsAsset /BRCosmetics/Characters/Player/Male/Medium/Bodies/M_MED_Tomato_Knight/Meshes/M_MED_Tomato_Knight_Physics_Pouches
bodies {
 dyn_pouch_01 [{shapeType=Capsule, length=8, radius=3.25, center=<1.6849 0.0258 0.8503>}]
 dyn_pouch_01_mstr [{shapeType=Capsule, length=0.505, radius=0.505}]
 dyn_pouch_02 [{shapeType=Capsule, length=8, radius=3.25, center=<1.9692 0.0122 0.5747>}]
 dyn_pouch_02_mstr [{shapeType=Capsule, length=0.505, radius=0.505}]
 dyn_pouch_03 [{shapeType=Capsule, length=8, radius=3.25, center=<2.0005 0.1357 0.5167>}]
 dyn_pouch_03_mstr [{shapeType=Capsule, length=0.505, radius=0.505}]
 pelvis [{shapeType=Capsule, length=1, radius=1}]
}
constraints [
 {bone1Name=dyn_pouch_01, bone2Name=dyn_pouch_01_mstr, swing1LimitDegrees=3, swing2LimitDegrees=3, twistLimitDegrees=2}
 {bone1Name=dyn_pouch_02, bone2Name=dyn_pouch_02_mstr, swing1LimitDegrees=3, swing2LimitDegrees=3, twistLimitDegrees=2}
 {bone1Name=dyn_pouch_03, bone2Name=dyn_pouch_03_mstr, swing1LimitDegrees=3, swing2LimitDegrees=3, twistLimitDegrees=2}
 {
  bone1Name=dyn_pouch_01_mstr
  bone2Name=pelvis
  swing1LimitDegrees=45
  swing1Motion=Free
  swing2LimitDegrees=45
  swing2Motion=Free
  twistLimitDegrees=45
  twistMotion=Free
 }
 {
  bone1Name=dyn_pouch_02_mstr
  bone2Name=pelvis
  swing1LimitDegrees=45
  swing1Motion=Free
  swing2LimitDegrees=45
  swing2Motion=Free
  twistLimitDegrees=45
  twistMotion=Free
 }
 {
  bone1Name=dyn_pouch_03_mstr
  bone2Name=pelvis
  swing1LimitDegrees=45
  swing1Motion=Free
  swing2LimitDegrees=45
  swing2Motion=Free
  twistLimitDegrees=45
  twistMotion=Free
 }
]
