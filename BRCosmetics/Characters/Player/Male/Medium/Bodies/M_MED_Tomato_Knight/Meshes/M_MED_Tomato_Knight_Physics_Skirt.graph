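# Physics asset that simulates the skirt of the male medium Tomato Knight player character body.
# Back, front and side chains of thin boxes per side use tight swing and twist limits between segments.
# Chains hang from the pelvis, some through small master-bone capsules with free rotation.
# Free-rotation constraints also link left and right tips of front and back chains, while pelvis and thigh capsules supply body shapes.
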PhysicsAsset /BRCosmetics/Characters/Player/Male/Medium/Bodies/M_MED_Tomato_Knight/Meshes/M_MED_Tomato_Knight_Physics_Skirt
bodies {
 dyn_skirtBack_01_l [{shapeType=Box, extentX=9, extentY=10, extentZ=2, center=<4.5 0 0>}]
 dyn_skirtBack_01_r [{shapeType=Box, extentX=9, extentY=10, extentZ=2, center=<4.5 0 0>}]
 dyn_skirtBack_02_l [{shapeType=Box, extentX=9, extentY=10, extentZ=2, center=<4.5 0 0>}]
 dyn_skirtBack_02_r [{shapeType=Box, extentX=9, extentY=10, extentZ=2, center=<4.5 0 0>}]
 dyn_skirtBack_03_l [{shapeType=Box, extentX=9, extentY=10, extentZ=2, center=<4.5 0 0>}]
 dyn_skirtBack_03_r [{shapeType=Box, extentX=9, extentY=10, extentZ=2, center=<4.5 0 0>}]
 dyn_skirtFront_01_l [{shapeType=Box, extentX=7, extentY=9, extentZ=1, center=<3.5 0 0>}]
 dyn_skirtFront_01_r [{shapeType=Box, extentX=7, extentY=9, extentZ=1, center=<3.5 0 0>}]
 dyn_skirtFront_02_l [{shapeType=Box, extentX=7, extentY=9, extentZ=1, center=<3.5 0 0>}]
 dyn_skirtFront_02_r [{shapeType=Box, extentX=7, extentY=9, extentZ=1, center=<3.5 0 0>}]
 dyn_skirtFront_03_l [{shapeType=Box, extentX=7, extentY=9, extentZ=2, center=<3.5 0 0>}]
 dyn_skirtFront_03_r [{shapeType=Box, extentX=7, extentY=9, extentZ=2, center=<3.5 0 0>}]
 dyn_skirtFront_04_l [{shapeType=Box, extentX=7, extentY=9, extentZ=2, center=<3.5 0 0>}]
 dyn_skirtFront_04_r [{shapeType=Box, extentX=7, extentY=9, extentZ=2, center=<3.5 0 0>}]
 dyn_skirtFront_mstr_l [{shapeType=Capsule, length=0.505, radius=0.505}]
 dyn_skirtFront_mstr_r [{shapeType=Capsule, length=0.505, radius=0.505}]
 dyn_skirtSide_01_l [{shapeType=Box, extentX=6, extentY=15, extentZ=2, center=<3 0 0>}]
 dyn_skirtSide_01_r [{shapeType=Box, extentX=6, extentY=15, extentZ=2, center=<3 0 0>}]
 dyn_skirtSide_02_l [{shapeType=Box, extentX=6, extentY=15, extentZ=2, center=<3 0 0>}]
 dyn_skirtSide_02_r [{shapeType=Box, extentX=6, extentY=15, extentZ=2, center=<3 0 0>}]
 dyn_skirtSide_03_l [{shapeType=Box, extentX=6, extentY=15, extentZ=2, center=<3 0 0>}]
 dyn_skirtSide_03_r [{shapeType=Box, extentX=6, extentY=15, extentZ=2, center=<3 0 0>}]
 dyn_skirtSide_mstr_l [{shapeType=Capsule, length=0.505, radius=0.505}]
 pelvis [{shapeType=Capsule, length=8.4612, radius=11, center=<-4.3792 -0.7139 0.0002>}]
 thigh_l [{shapeType=Capsule, length=25, radius=9, center=<-17.6529 0.0006 0.3133>, rotation=<rot 90 0 0>}]
 thigh_r [{shapeType=Capsule, length=25, radius=9, center=<17.6529 -0.0006 -0.5619>, rotation=<rot 90 0 -179.9999>}]
}
constraints [
 {bone1Name=dyn_skirtBack_01_l, bone2Name=pelvis, swing1LimitDegrees=25, swing2LimitDegrees=25, twistLimitDegrees=5}
 {bone1Name=dyn_skirtBack_02_l, bone2Name=dyn_skirtBack_01_l, swing1LimitDegrees=25, swing2LimitDegrees=25, twistLimitDegrees=5}
 {bone1Name=dyn_skirtBack_03_l, bone2Name=dyn_skirtBack_02_l, swing1LimitDegrees=25, swing2LimitDegrees=25, twistLimitDegrees=5}
 {bone1Name=dyn_skirtBack_01_r, bone2Name=pelvis, swing1LimitDegrees=25, swing2LimitDegrees=25, twistLimitDegrees=5}
 {bone1Name=dyn_skirtBack_02_r, bone2Name=dyn_skirtBack_01_r, swing1LimitDegrees=25, swing2LimitDegrees=25, twistLimitDegrees=5}
 {bone1Name=dyn_skirtBack_03_r, bone2Name=dyn_skirtBack_02_r, swing1LimitDegrees=25, swing2LimitDegrees=25, twistLimitDegrees=5}
 {bone1Name=dyn_skirtSide_01_l, bone2Name=dyn_skirtSide_mstr_l, swing1LimitDegrees=25, swing2LimitDegrees=25, twistLimitDegrees=5}
 {bone1Name=dyn_skirtSide_02_l, bone2Name=dyn_skirtSide_01_l, swing1LimitDegrees=25, swing2LimitDegrees=25, twistLimitDegrees=5}
 {bone1Name=dyn_skirtSide_03_l, bone2Name=dyn_skirtSide_02_l, swing1LimitDegrees=25, swing2LimitDegrees=25, twistLimitDegrees=5}
 {bone1Name=dyn_skirtSide_01_r, bone2Name=pelvis, swing1LimitDegrees=25, swing2LimitDegrees=25, twistLimitDegrees=5}
 {bone1Name=dyn_skirtSide_02_r, bone2Name=dyn_skirtSide_01_r, swing1LimitDegrees=25, swing2LimitDegrees=25, twistLimitDegrees=5}
 {bone1Name=dyn_skirtSide_03_r, bone2Name=dyn_skirtSide_02_r, swing1LimitDegrees=25, swing2LimitDegrees=25, twistLimitDegrees=5}
 {bone1Name=dyn_skirtFront_01_r, bone2Name=dyn_skirtFront_mstr_r, swing1LimitDegrees=25, swing2LimitDegrees=25, twistLimitDegrees=5}
 {bone1Name=dyn_skirtFront_02_r, bone2Name=dyn_skirtFront_01_r, swing1LimitDegrees=25, swing2LimitDegrees=25, twistLimitDegrees=5}
 {bone1Name=dyn_skirtFront_03_r, bone2Name=dyn_skirtFront_02_r, swing1LimitDegrees=25, swing2LimitDegrees=25, twistLimitDegrees=5}
 {bone1Name=dyn_skirtFront_04_r, bone2Name=dyn_skirtFront_03_r, swing1LimitDegrees=25, swing2LimitDegrees=25, twistLimitDegrees=5}
 {bone1Name=dyn_skirtFront_01_l, bone2Name=dyn_skirtFront_mstr_l, swing1LimitDegrees=25, swing2LimitDegrees=25, twistLimitDegrees=5}
 {bone1Name=dyn_skirtFront_02_l, bone2Name=dyn_skirtFront_01_l, swing1LimitDegrees=25, swing2LimitDegrees=25, twistLimitDegrees=5}
 {bone1Name=dyn_skirtFront_03_l, bone2Name=dyn_skirtFront_02_l, swing1LimitDegrees=25, swing2LimitDegrees=25, twistLimitDegrees=5}
 {bone1Name=dyn_skirtFront_04_l, bone2Name=dyn_skirtFront_03_l, swing1LimitDegrees=25, swing2LimitDegrees=25, twistLimitDegrees=5}
 {
  bone1Name=thigh_l
  bone2Name=pelvis
  swing1LimitDegrees=45
  swing1Motion=Free
  swing2LimitDegrees=45
  swing2Motion=Free
  twistLimitDegrees=45
  twistMotion=Free
 }
 {
  bone1Name=thigh_r
  bone2Name=pelvis
  swing1LimitDegrees=45
  swing1Motion=Free
  swing2LimitDegrees=45
  swing2Motion=Free
  twistLimitDegrees=45
  twistMotion=Free
 }
 {
  bone1Name=dyn_skirtBack_03_l
  bone2Name=dyn_skirtBack_03_r
  swing1LimitDegrees=45
  swing1Motion=Free
  swing2LimitDegrees=45
  swing2Motion=Free
  twistLimitDegrees=45
  twistMotion=Free
 }
 {
  bone1Name=dyn_skirtFront_04_l
  bone2Name=dyn_skirtFront_04_r
  swing1LimitDegrees=45
  swing1Motion=Free
  swing2LimitDegrees=45
  swing2Motion=Free
  twistLimitDegrees=45
  twistMotion=Free
 }
 {
  bone1Name=dyn_skirtSide_mstr_l
  bone2Name=pelvis
  swing1LimitDegrees=45
  swing1Motion=Free
  swing2LimitDegrees=45
  swing2Motion=Free
  twistLimitDegrees=45
  twistMotion=Free
 }
 {
  bone1Name=dyn_skirtFront_mstr_l
  bone2Name=pelvis
  swing1LimitDegrees=45
  swing1Motion=Free
  swing2LimitDegrees=45
  swing2Motion=Free
  twistLimitDegrees=45
  twistMotion=Free
 }
 {
  bone1Name=dyn_skirtFront_mstr_r
  bone2Name=pelvis
  swing1LimitDegrees=45
  swing1Motion=Free
  swing2LimitDegrees=45
  swing2Motion=Free
  twistLimitDegrees=45
  twistMotion=Free
 }
]
