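# Physics asset for the plume on the Tomato Knight medium male character skin, simulating the plume's motion.
# It has two capsule bodies: a large capsule on dyn_plume, offset along the bone, and a tiny one on the dyn_plume_mstr bone.
# A single constraint joins them with tight swing and twist limits, so the plume can only sway slightly relative to its master bone.
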PhysicsAsset /BRCosmetics/Characters/Player/Male/Medium/Bodies/M_MED_Tomato_Knight/Meshes/Parts/M_MED_Tomato_Knight_Physics_Plume
bodies {dyn_plume [{shapeType=Capsule, length=20, radius=10, center=<7.5 0 0>}], dyn_plume_mstr [{shapeType=Capsule, length=0.505, radius=0.505}]}
constraints [{bone1Name=dyn_plume, bone2Name=dyn_plume_mstr, swing1LimitDegrees=10, swing2LimitDegrees=5, twistLimitDegrees=2}]
