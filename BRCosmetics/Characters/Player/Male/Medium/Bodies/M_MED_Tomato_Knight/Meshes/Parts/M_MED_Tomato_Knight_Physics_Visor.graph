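# Physics asset for the visor part of the male medium Tomato Knight player cosmetic body.
# A box body on the dyn_visor bone joins a small capsule body on dyn_visor_mstr through a single constraint.
# The constraint locks swing1 and twist, leaving swing2 as the only rotational axis not locked, so the visor moves like a hinge.
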PhysicsAsset /BRCosmetics/Characters/Player/Male/Medium/Bodies/M_MED_Tomato_Knight/Meshes/Parts/M_MED_Tomato_Knight_Physics_Visor
bodies {
 dyn_visor [{shapeType=Box, extentX=30, extentY=36, extentZ=10, center=<15 0 0>}]
 dyn_visor_mstr [{shapeType=Capsule, length=0.505, radius=0.505}]
}
constraints [{
  bone1Name=dyn_visor
  bone2Name=dyn_visor_mstr
  swing1LimitDegrees=45
  swing1Motion=Locked
  swing2LimitDegrees=5
  twistLimitDegrees=15
  twistMotion=Locked
 }]
